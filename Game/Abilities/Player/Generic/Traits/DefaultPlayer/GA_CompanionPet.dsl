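; GA_CompanionPet is a default-player gameplay ability whose name points to companion petting, lasting through a tagged player-companion interaction.
; On activation it runs the parent logic, then waits for the PlayerInteractingWithCompanion tag to be added to the avatar.
; Once added, it waits for that tag's removal and ends the ability locally.
; On end it cancels only the tag-added wait; the stored removal wait is not cancelled.
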(event Ability|EventActivateAbility
  (bind _returnvalue (Ability|GetAvatarActorfromActorInfo))
  (|CallParentFunction)
  (bind _asynctaskproxy (Ability|Tasks|WaitGameplayTagAddToActor _returnvalue "(TagName=\"Companion.Interaction.PlayerInteractingWithCompanion\")" true)
    (:then
      (Variables|Default|SetAddAsyncActions _asynctaskproxy))
    (:Added
      (bind _asynctaskproxy_2 (Ability|Tasks|WaitGameplayTagRemoveFromActor _returnvalue "(TagName=\"Companion.Interaction.PlayerInteractingWithCompanion\")" true)
        (:then
          (Variables|Default|SetRemoveAsyncActions _asynctaskproxy_2))
        (:Removed
          (Ability|EndAbilityLocally))))))

(event Ability|EventOnEndAbility
  (params (bWasCancelled :type Boolean))
  (AsyncAction|Cancel (Variables|Default|GetAddAsyncActions)))
